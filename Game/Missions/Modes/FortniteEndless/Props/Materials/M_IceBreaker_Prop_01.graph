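# Translucent, lit material for the IceBreaker ice block prop in FortniteEndless, blending from intact to damaged looks via a DamageState scalar.
# DamageState lerps diffuse, normal, packed metallic/specular/roughness masks, emissive and opacity, while a power on the damaged emissive's alpha erodes its glow.
# Depth-masked, bump-offset clouds fake an interior, shifting color only in the last third of damage.
# Fresnel terms shape glow, cloud visibility and opacity, alongside a refraction parameter.
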Material /Game/Missions/Modes/FortniteEndless/Props/Materials/M_IceBreaker_Prop_01
graph {
 properties {
  blendMode=BLEND_Translucent
  pixelDepthOffsetMode=PDOM_Legacy
  refractionMethod=RM_IndexOfRefraction
  translucencyLightingMode=TLM_Surface
  usedShadingModels=Default Lit
 }
 nodes [
  n1 = Clamp(None=Multiply(
    A=ScalarParameter() {group=Gameplay, parameterName=DamageState, sliderMax=1}
    B=ScalarParameter() {defaultValue=1, group=Gameplay, parameterName=DamageState_OffsetMultiplier}
   ))
  n2 = TextureSampleParameter2D() {parameterName=EmissiveTexture-Damaged, texture=/Game/Missions/Modes/FortniteEndless/Props/Textures/T_IceBreakerBlockStage2_E}
  n3 = ScalarParameter() {defaultValue=10, parameterName=ErodePowerMax}
  n4 = Clamp(None=Power(Base=@n2.A, Exp=LinearInterpolate(A=@n3, Alpha=@n1) {constA=10}) {constExponent=1})
  n5 = LinearInterpolate(
   A=LinearInterpolate(
    A=VectorParameter() {parameterName=EmmissiveColor, defaultValue=<rgba 0 0 0 1>}.RGB
    B=VectorParameter() {parameterName=EmmissiveColor-Damaged, defaultValue=<rgba 0 0 0 1>}.RGB
    Alpha=@n1
   )
   B=LinearInterpolate(
    A=TextureSampleParameter2D() {parameterName=EmissiveTexture, texture=&T_BlacksRGB}.RGB
    B=Multiply(A=@n2.RGB, B=@n4)
    Alpha=@n1
   )
   Alpha=ScalarParameter() {parameterName=UseEmissiveTexture}
  )
  n6 = ScalarParameter() {defaultValue=1, parameterName=EmissiveMultiplier}
  n7 = MaterialFunctionCall(
   Brightness=ScalarParameter() {defaultValue=1, group=FresnelGlow, parameterName=FresnelGlow_Hardness}
   Exponent=ScalarParameter() {defaultValue=2, group=FresnelGlow, parameterName=FresnelGlow_Exponent}
  ) {materialFunction=&MF_Fresnel_FX}
  n8 = LinearInterpolate(
   A=VectorParameter() {group=FresnelGlow, parameterName=FresnelGlow-Color, defaultValue=<rgba 0 0 0 1>}.RGB
   B=VectorParameter() {group=FresnelGlow, parameterName=FresnelGlow-Color-Damaged, defaultValue=<rgba 0 0 0 1>}.RGB
   Alpha=@n1
  )
  n9 = VectorParameter() {group=InsideElements, parameterName=Color-Clouds, defaultValue=<rgba 0 0.04 0.1 1>}
  n10 = VectorParameter() {group=InsideElements, parameterName=Color-Clouds-Damaged, defaultValue=<rgba 0 2 5 1>}
  n11 = Clamp(None=Subtract(A=Multiply(A=@n1) {constB=3}) {constB=2})
  n12 = TextureSample() {samplerSource=SSM_Clamp_WorldGroupSettings, texture=/Game/Missions/Modes/FortniteEndless/Props/Textures/T_IceBreakerBlock_DepthMask}
  n13 = BumpOffset(
   Coordinate=TextureCoordinate() {uTiling=2, vTiling=2}
   Height=Multiply(A=Power(Base=@n12.R), B=ScalarParameter() {defaultValue=-2, group=InsideElements, parameterName=Height})
  ) {heightRatio=0.1, referencePlane=1}
  n14 = Clamp(None=Power(
    Base=TextureSample(UVs=@n13) {samplerSource=SSM_Wrap_WorldGroupSettings, texture=/Game/Athena/SafeZone/S11/T_TilingClouds_D}.RGB
    Exp=ScalarParameter() {defaultValue=1, group=InsideElements, parameterName=Cloud-Exp}
   ))
  n15 = ScalarParameter() {defaultValue=1, group=InsideElements, parameterName=Cloud-Hardness}
  n16 = Multiply(A=@n12.RGB, B=Clamp(None=Multiply(A=@n14, B=@n15)))
  n17 = MaterialFunctionCall(
   Brightness=ScalarParameter() {defaultValue=1, group=InsideElements, parameterName=CloudFresnel_Brightness}
   Exponent=ScalarParameter() {defaultValue=2, group=InsideElements, parameterName=CloudFresnel_Exponent}
   Vertex / Pixel Normal=VertexNormalWS()
  ) {materialFunction=&MF_Fresnel_FX}
  Add(
   A=Add(A=Multiply(A=@n5, B=@n6), B=Multiply(A=@n7.Fresnel Inverted, B=@n8)) {constB=0}
   B=Multiply(A=LinearInterpolate(A=@n9.RGB, B=@n10.RGB, Alpha=@n11), B=Multiply(A=@n16, B=@n17.Fresnel))
  ) {constB=0}
  n18 = LinearInterpolate(
   A=LinearInterpolate(
    A=ScalarParameter() {parameterName=OpacityA}
    B=ScalarParameter() {defaultValue=0.8, parameterName=Opacity-DamagedState, sliderMax=1}
    Alpha=@n1
   )
   B=ScalarParameter() {defaultValue=1, parameterName=OpacityB}
   Alpha=MaterialFunctionCall(
    Brightness=ScalarParameter() {defaultValue=3, parameterName=FresnelOpacity_Hardness}
    Exponent=ScalarParameter() {defaultValue=0.5, parameterName=FresnelOpacity_Exponent}
    Vertex / Pixel Normal=VertexNormalWS()
   ) {materialFunction=&MF_Fresnel_FX}.Fresnel Inverted
  )
  Add(A=@n18, B=Power(Base=OneMinus(None=@n16)) {constExponent=6})
  n19 = ScalarParameter() {defaultValue=0.5, parameterName=AlphaMask_Influence, sliderMax=1}
  n20 = TextureSampleParameter2D() {parameterName=Diffuse, texture=/Game/Missions/Modes/FortniteEndless/Props/Textures/T_IceBreakerBlock_D}
  Clamp(None=Add(A=@n18, B=Multiply(A=@n20.A, B=@n19)))
  n21 = TextureSampleParameter2D() {parameterName=Texture_S, samplerType=SAMPLERTYPE_Masks, texture=/Game/Missions/Modes/FortniteEndless/Props/Textures/T_IceBreakerBlock_S}
  n22 = TextureSampleParameter2D() {
   parameterName=Texture_S-Damaged
   samplerType=SAMPLERTYPE_Masks
   texture=/Game/Missions/Modes/FortniteEndless/Props/Textures/T_IceBreakerBlockStage2_S
  }
  n23 = MaterialFunctionCall(Color=LinearInterpolate(A=@n21.RGBA, B=@n22.RGBA, Alpha=@n1)) {materialFunction=&SplitComponents}
  LinearInterpolate(A=ScalarParameter() {parameterName=Metallic_Min}, B=ScalarParameter() {defaultValue=1, parameterName=Metallic_Max}, Alpha=@n23.G)
  LinearInterpolate(A=ScalarParameter() {parameterName=Specular_Min}, B=ScalarParameter() {defaultValue=1, parameterName=Specular_Max}, Alpha=@n23.R)
  LinearInterpolate(A=ScalarParameter() {parameterName=Roughness_Min}, B=ScalarParameter() {defaultValue=1, parameterName=Roughness_Max}, Alpha=@n23.B)
  Multiply(
   A=LinearInterpolate(
    A=@n20.RGB
    B=TextureSampleParameter2D() {parameterName=Diffuse-Damaged, texture=/Game/Missions/Modes/FortniteEndless/Props/Textures/T_IceBreakerBlockStage2_D}.RGB
    Alpha=@n1
   )
   B=ScalarParameter() {defaultValue=1, parameterName=DiffuseMultiply}
  )
  n24 = TextureSampleParameter2D() {parameterName=NormalMap, samplerType=SAMPLERTYPE_Normal, texture=/Game/Missions/Modes/FortniteEndless/Props/Textures/T_IceBreakerBlock_N}
  n25 = TextureSampleParameter2D() {
   parameterName=NormalMap-Damaged
   samplerType=SAMPLERTYPE_Normal
   texture=/Game/Missions/Modes/FortniteEndless/Props/Textures/T_IceBreakerBlockStage2_N
  }
  Reroute(None=LinearInterpolate(A=@n24.RGB, B=@n25.RGB, Alpha=@n1))
  ScalarParameter() {defaultValue=0.9, parameterName=Refraction}
 ]
}
